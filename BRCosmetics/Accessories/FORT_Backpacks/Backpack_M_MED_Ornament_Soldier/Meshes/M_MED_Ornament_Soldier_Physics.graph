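# Physics asset for the Ornament Soldier backpack cosmetic, giving its bow and ribbon collision and constrained motion relative to the pack.
# Spheres cover the pack, the bow and its two sub-bones, while a thin box covers the ribbon.
# The bow and ribbon constrain directly to the pack with wide swing and narrow twist, the bow's second swing axis being locked at zero.
# Both bow sub-bones hang off the main bow with very tight limits.
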PhysicsAsset /BRCosmetics/Accessories/FORT_Backpacks/Backpack_M_MED_Ornament_Soldier/Meshes/M_MED_Ornament_Soldier_Physics
bodies {
 dyn_bow [{shapeType=Sphere, radius=6.2715, center=<2 0 0>}]
 dyn_bow_a [{shapeType=Sphere, radius=3.9924, center=<2 0 0>}]
 dyn_bow_b [{shapeType=Sphere, radius=3.4868, center=<2 0 0>}]
 dyn_ribbon [{shapeType=Box, extentX=0.7972, extentY=4.1967, extentZ=8.8556, center=<4.3984 0 0>, rotation=<rot 90 0 0>}]
 pack [{shapeType=Sphere, radius=9.4201, center=<-0.6416 -7.4241 -0.5577>}]
}
constraints [
 {bone1Name=dyn_bow, bone2Name=pack, swing1LimitDegrees=45, swing2LimitDegrees=0, twistLimitDegrees=5}
 {bone1Name=dyn_bow_b, bone2Name=dyn_bow, swing1LimitDegrees=2, swing2LimitDegrees=2, twistLimitDegrees=10}
 {bone1Name=dyn_bow_a, bone2Name=dyn_bow, swing1LimitDegrees=2, swing2LimitDegrees=2, twistLimitDegrees=10}
 {bone1Name=dyn_ribbon, bone2Name=pack, swing1LimitDegrees=45, swing2LimitDegrees=45, twistLimitDegrees=5}
]
